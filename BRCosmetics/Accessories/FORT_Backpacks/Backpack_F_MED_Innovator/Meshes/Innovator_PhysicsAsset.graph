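# Physics asset for the Innovator female medium backpack cosmetic, giving its skeletal mesh a simple collision and physics representation.
# It holds a single body on Backpack_Innovator_LOD0, shaped as one rotated, offset capsule.
# It defines no constraints, so there is no jointed or ragdoll chain.
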PhysicsAsset /BRCosmetics/Accessories/FORT_Backpacks/Backpack_F_MED_Innovator/Meshes/Innovator_PhysicsAsset
bodies {Backpack_Innovator_LOD0 [{shapeType=Capsule, length=20.8834, radius=20.5526, center=<-0.8984 -7.2607 -4.592>, rotation=<rot 1.5906 -90 -115.8883>}]}
constraints []
